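# Simple river water preview material for Landmass, using the Single Layer Water shading model and usable on spline meshes.
# Tiled UVs and a constant flow vector drive a simple flow-map water normal, softened by a flatness control.
# Scattering color is scaled by its alpha, absorption is the inverse square of its color parameter, and anisotropy sets the phase.
# A pixel-depth ramp scaled by parameter R builds a value for pixel-normal-offset refraction.
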Material /Landmass/PreviewContent/Water/Mats/M_Simple_River_01
graph {
 properties {
  bUsedWithSplineMeshes=true
  pixelDepthOffsetMode=PDOM_Legacy
  refractionMethod=RM_PixelNormalOffset
  shadingModel=MSM_SingleLayerWater
  translucencyLightingMode=TLM_Surface
  usedShadingModels=SingleLayerWater
 }
 nodes [
  n1 = Divide(B=PixelDepth().R) {constA=100, constB=100}
  Add(B=Multiply(A=Saturate(None=@n1), B=ScalarParameter() {defaultValue=0.1, parameterName=R})) {constA=1}
  n2 = TextureCoordinate()
  n3 = ComponentMask(None=VectorParameter() {parameterName=Water Tiling, defaultValue=<rgba 2 1 0 0>}.RGB) {g=true, r=true}
  MaterialFunctionCall(
   Normal=MaterialFunctionCall!(
    Diffuse
    Normal Map=TextureObject() {samplerType=SAMPLERTYPE_Normal, texture=/Engine/Functions/Engine_MaterialFunctions02/ExampleContent/Textures/water_n}
    "Flow Vector Map (see tooltip)"=VectorParameter() {parameterName=Flow, defaultValue=<rgba 0.5 0 0 0>}.RGB
    UV texture 2 offset
    "Time (see tooltip)"
    UVs=Multiply(A=@n2, B=@n3)
   ) {materialFunction=/Engine/Functions/Engine_MaterialFunctions02/Texturing/FlowMaps_Simple}.Normal
   Flatness=ScalarParameter() {parameterName=Normal Flatness}
  ) {materialFunction=&FlattenNormal}
  ScalarParameter() {parameterName=Opacity}
  ScalarParameter() {defaultValue=0.225, parameterName=Spec}
  ScalarParameter() {defaultValue=0.05, parameterName=Roughness}
  n4 = VectorParameter() {group=Water Shading Model, parameterName=Scattering, defaultValue=<rgba 0.1 0.1 0.1 1>}
  n5 = VectorParameter() {group=Water Shading Model, parameterName=Absorption, defaultValue=<rgba 1 8 20 100>}
  SingleLayerWaterMaterialOutput!(
   ScatteringCoefficients=Multiply(A=@n4.RGB, B=Divide(A=@n4.A) {constB=100})
   AbsorptionCoefficients=Divide(B=Multiply(A=@n5.RGB, B=@n5.RGB))
   PhaseG=ScalarParameter() {defaultValue=0.4, parameterName=Aniso}
   ColorScaleBehindWater
  )
  VectorParameter() {parameterName=Base}
 ]
}
